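# Translucent unlit UI material for the Athena HUD inventory background, drawing dark navy side glows with a grid overlay.
# Pairs of exponential radial gradients centered beyond the left and right edges use vertically rescaled UVs to form a broad mask and a tighter Power-parameter mask.
# These masks blend two blues and also build a separate scalar blend, likely for opacity.
# The invGrid texture tiles every 64 units of texcoord 3 and adds brighter blue within the sharpened broad mask.
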Material /FortniteCoreUI/Athena/HUD/Inventory/Styles/M_InventoryBackground
graph {
 properties {
  bUsedWithStaticMesh=false
  blendMode=BLEND_Translucent
  materialDomain=MD_UI
  pixelDepthOffsetMode=PDOM_Legacy
  shadingModel=MSM_Unlit
  usedShadingModels=Unlit
 }
 nodes [
  n1 = Constant() {r=1}
  n2 = Constant() {r=0.5}
  n3 = Subtract(B=@n2)
  n4 = Add(A=Multiply(A=TextureCoordinate(), B=AppendVector(A=@n1, B=@n2)), B=AppendVector(A=Constant(), B=Multiply(A=@n3) {constB=0.5})) {constB=0}
  n5 = Constant() {r=0.7}
  n6 = Constant() {r=2}
  n7 = Add(
   A=MaterialFunctionCall(UVs=@n4, CenterPosition=Constant2Vector() {g=0.5, r=-0.125}, Radius=@n5, Density=@n6) {materialFunction=&RadialGradientExponential}.RadialGradientExponential
   B=MaterialFunctionCall(UVs=@n4, CenterPosition=Constant2Vector() {g=0.5, r=1.125}, Radius=@n5, Density=@n6) {materialFunction=&RadialGradientExponential}.RadialGradientExponential
  )
  n8 = Power(Base=@n7) {constExponent=0.5}
  n9 = Constant() {r=1}
  n10 = Constant() {r=0.3}
  n11 = Subtract(B=@n10)
  n12 = Add(A=Multiply(A=TextureCoordinate(), B=AppendVector(A=@n9, B=@n10)), B=AppendVector(A=Constant(), B=Multiply(A=@n11) {constB=0.5})) {constB=0}
  n13 = Constant() {r=0.23}
  n14 = Constant() {r=3}
  n15 = MaterialFunctionCall(UVs=@n12, CenterPosition=Constant2Vector() {g=0.5, r=-0.125}, Radius=@n13, Density=@n14) {materialFunction=&RadialGradientExponential}
  n16 = MaterialFunctionCall(UVs=@n12, CenterPosition=Constant2Vector() {g=0.5, r=1.125}, Radius=@n13, Density=@n14) {materialFunction=&RadialGradientExponential}
  n17 = Clamp(None=Power(Base=Add(A=@n15.RadialGradientExponential, B=@n16.RadialGradientExponential), Exp=ScalarParameter() {defaultValue=2.5, parameterName=Power}) {constExponent=9})
  LinearInterpolate(A=LinearInterpolate(Alpha=Clamp(None=@n8)), Alpha=@n17)
  n18 = Constant4Vector() {constant=<rgba 0.0024 0.0742 0.1714 1>}
  n19 = TextureCoordinate() {coordinateIndex=3}
  n20 = Constant2Vector() {g=64, r=64}
  n21 = TextureSample(UVs=Multiply(A=Divide(A=@n19, B=@n20), B=TextureCoordinate())) {texture=&/invGrid}
  n22 = Clamp(None=Power(Base=@n7) {constExponent=11})
  LinearInterpolate(
   A=LinearInterpolate(A=Constant4Vector() {constant=<rgba 0.0012 0.0061 0.016 1>}.RGBA, B=@n18.RGBA, Alpha=@n17)
   B=Multiply(B=@n18.RGBA) {constA=1.125}
   Alpha=Clamp(None=Multiply(A=@n21.A, B=@n22))
  )
 ]
}
